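# Animation Blueprint for the LiquidCouch two-handed pickaxe, adding physics motion and corrective bone offsets.
# Each update mirrors the attach parent's root, pelvis, thigh and calf world transforms onto the axe's body bones, then offsets body_root.
# That pose feeds two zero-gravity RigidBody branches with separate physics assets, selected by whether the axe is on a pawn.
# GravityOverride and RBAN Alpha are set but do not drive the RigidBody nodes.
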AnimBlueprint /BRCosmetics/Weapons/FORT_Melee/Pickaxe_LiquidCouch/Meshes/LiquidCouch_Axe_AnimBP
target_skeleton=&/LiquidCouch_Axe_Skeleton
graphs [
 AnimGraph {nodes [
   n1 = "Transform (Modify) Bone - Bone: arm_r_01"(
    ComponentPose=Local To Component(LocalPose=Use cached pose 'Pre Dynamics setup'())
    Alpha=1
    Translation="0,0,0"
    Rotation="20,30,0"
    Scale="1,1,1"
   ) {
    rotationMode=BMM_Additive
    rotationSpace=BCS_BoneSpace
    translationMode=BMM_Additive
    translationSpace=BCS_BoneSpace
    boneToModify {boneName=arm_r_01}
    rotation {pitch=20, yaw=30}
   }
   n2 = RigidBody!(ComponentPose=@n1, Alpha, OverrideWorldGravity:Vector="0,0,0", ExternalForce) -> (Pose) {
    bClampLinearTranslationLimitToRefPose=true
    bOverrideWorldGravity=true
    overridePhysicsAsset=&/LiquidCouch_Axe_FEoffPawn_Physics
    simulationSpace=BaseBoneSpace
    alphaBoolBlend {blendOption=Linear}
    baseBoneRef {boneName=Melee_TwoHanded_Handle}
    componentAppliedLinearAccClamp=<0 0 0>
    simSpaceSettings {
     maxAngularAcceleration=40
     maxAngularVelocity=8
     maxLinearAcceleration=3500
     maxLinearVelocity=700
     velocityScaleZ=0.5
     worldAlpha=0.425
     externalLinearDragV=<1 1 1>
    }
   }
   n3 = "Transform (Modify) Bone - Bone: Handle_Corrective_up"(
    ComponentPose=Local To Component(LocalPose=Use cached pose 'Pre Dynamics setup'())
    Alpha=1
    Translation="-0.284,0.185,0"
    Rotation="0,0,0"
    Scale="1,0.5,1"
   ) {
    rotationMode=BMM_Additive
    rotationSpace=BCS_BoneSpace
    scaleMode=BMM_Additive
    scaleSpace=BCS_BoneSpace
    translationMode=BMM_Additive
    translationSpace=BCS_BoneSpace
    boneToModify {boneName=Handle_Corrective_up}
    scale {y=0.5}
    translation=<-0.284 0.185>
   }
   n4 = "Transform (Modify) Bone - Bone: Handle_Corrective_down"(ComponentPose=@n3, Alpha=1, Translation="-0.5,0.7,0", Rotation="0,0,3.959", Scale="0.5,0.5,1") {
    rotationMode=BMM_Additive
    rotationSpace=BCS_BoneSpace
    scaleMode=BMM_Additive
    scaleSpace=BCS_BoneSpace
    translationMode=BMM_Additive
    translationSpace=BCS_BoneSpace
    boneToModify {boneName=Handle_Corrective_down}
    rotation {roll=3.959}
    scale=<0.5 0.5>
    translation=<-0.5 0.7>
   }
   n5 = RigidBody(ComponentPose=@n4, OverrideWorldGravity="0,0,0") {
    bClampLinearTranslationLimitToRefPose=true
    bOverrideWorldGravity=true
    overridePhysicsAsset=&/LiquidCouch_Axe_Physics
    simulationSpace=BaseBoneSpace
    alphaBoolBlend {blendOption=Linear}
    baseBoneRef {boneName=Melee_TwoHanded_Handle}
    componentAppliedLinearAccClamp=<0 0 0>
    simSpaceSettings {
     maxAngularAcceleration=40
     maxAngularVelocity=8
     maxLinearAcceleration=3500
     maxLinearVelocity=700
     velocityScaleZ=0.5
     worldAlpha=0.425
     externalLinearDragV=<1 1 1>
    }
   }
   n6 = "Transform (Modify) Bone - Bone: base_leg_l_01"(ComponentPose=@n5, Alpha=1, Translation="0,2,0", Rotation="0,0,0", Scale="1,1,1") {
    rotationMode=BMM_Additive
    rotationSpace=BCS_BoneSpace
    scaleMode=BMM_Additive
    scaleSpace=BCS_BoneSpace
    translationMode=BMM_Additive
    translationSpace=BCS_BoneSpace
    boneToModify {boneName=base_leg_l_01}
    translation {y=2}
   }
   n7 = "Transform (Modify) Bone - Bone: Handle_Legs_L_corrective"(ComponentPose=@n6, Alpha=1, Translation="0,5,0", Rotation="0,0,0", Scale="1,1,1") {
    rotationMode=BMM_Additive
    rotationSpace=BCS_BoneSpace
    scaleMode=BMM_Additive
    scaleSpace=BCS_BoneSpace
    translationMode=BMM_Additive
    translationSpace=BCS_BoneSpace
    boneToModify {boneName=Handle_Legs_L_corrective}
    translation {y=5}
   }
   Output Pose(Result=Blend Poses by bool(
     BlendPose_0=Component To Local(ComponentPose=@n2)
     BlendPose_1=Component To Local(ComponentPose=@n7)
     bActiveValue=Get is Visible but NOT Equipped() -> (is Visible but NOT Equipped:bool)
    )) {layerGroup=None}
   AnimGraphNode_SubInput_0 = Input Pose()
   Save cached pose 'Start Pose'(Pose=@AnimGraphNode_SubInput_0) {cacheName=Start Pose}
   n8 = "Transform (Modify) Bone - Bone: body_root"(
    ComponentPose=Local To Component(LocalPose=Use cached pose 'Start Pose'())
    Alpha=1
    Translation=Get body_root_location() -> (body_root_location:Vector)
    Rotation=Get body_root_rotation() -> (body_root_rotation:Rotator)
    Scale="1,1,1"
   ) {
    rotationMode=BMM_Replace
    rotationSpace=BCS_WorldSpace
    scaleSpace=BCS_BoneSpace
    translationMode=BMM_Replace
    translationSpace=BCS_WorldSpace
    alphaBoolBlend {blendOption=Linear}
    alphaScaleBiasClamp {bInterpResult=true, interpSpeedDecreasing=1, interpSpeedIncreasing=1, inRange {max=0, min=-40}, outRange {max=0, min=1}}
    boneToModify {boneName=body_root}
   }
   n9 = "Transform (Modify) Bone - Bone: body_pelvis"(
    ComponentPose=@n8
    Alpha=1
    Translation=Get pelvis_location() -> (pelvis_location:Vector)
    Rotation=Get pelvis_rotation() -> (pelvis_rotation:Rotator)
    Scale="1,1,1"
   ) {
    rotationMode=BMM_Replace
    rotationSpace=BCS_WorldSpace
    scaleSpace=BCS_BoneSpace
    translationMode=BMM_Replace
    translationSpace=BCS_WorldSpace
    alphaBoolBlend {blendOption=Linear}
    alphaScaleBiasClamp {inRange {max=0, min=-40}, outRange {max=0, min=1}}
    boneToModify {boneName=body_pelvis}
   }
   n10 = "Transform (Modify) Bone - Bone: body_thigh_l"(
    ComponentPose=@n9
    Alpha=1
    Translation=Get thigh_left_location() -> (thigh_left_location:Vector)
    Rotation=Get thigh_left_rotation() -> (thigh_left_rotation:Rotator)
    Scale="1,1,1"
   ) {
    rotationMode=BMM_Replace
    rotationSpace=BCS_WorldSpace
    scaleSpace=BCS_BoneSpace
    translationMode=BMM_Replace
    translationSpace=BCS_WorldSpace
    alphaBoolBlend {blendOption=Linear}
    alphaScaleBiasClamp {inRange {max=0, min=-40}, outRange {max=0, min=1}}
    boneToModify {boneName=body_thigh_l}
   }
   n11 = "Transform (Modify) Bone - Bone: body_thigh_r"(
    ComponentPose=@n10
    Alpha=1
    Translation=Get thigh_right_location() -> (thigh_right_location:Vector)
    Rotation=Get thigh_right_rotation() -> (thigh_right_rotation:Rotator)
    Scale="1,1,1"
   ) {
    rotationMode=BMM_Replace
    rotationSpace=BCS_WorldSpace
    scaleSpace=BCS_BoneSpace
    translationMode=BMM_Replace
    translationSpace=BCS_WorldSpace
    alphaBoolBlend {blendOption=Linear}
    alphaScaleBiasClamp {inRange {max=0, min=-40}, outRange {max=0, min=1}}
    boneToModify {boneName=body_thigh_r}
   }
   n12 = "Transform (Modify) Bone - Bone: body_calf_l"(
    ComponentPose=@n11
    Alpha=1
    Translation=Get calf_left_location() -> (calf_left_location:Vector)
    Rotation=Get calf_left_rotation() -> (calf_left_rotation:Rotator)
    Scale="1,1,1"
   ) {
    rotationMode=BMM_Replace
    rotationSpace=BCS_WorldSpace
    scaleSpace=BCS_BoneSpace
    translationMode=BMM_Replace
    translationSpace=BCS_WorldSpace
    alphaBoolBlend {blendOption=Linear}
    alphaScaleBiasClamp {inRange {max=0, min=-40}, outRange {max=0, min=1}}
    boneToModify {boneName=body_calf_l}
   }
   n13 = "Transform (Modify) Bone - Bone: body_calf_r"(
    ComponentPose=@n12
    Alpha=1
    Translation=Get calf_right_location() -> (calf_right_location:Vector)
    Rotation=Get calf_right_rotation() -> (calf_right_rotation:Rotator)
    Scale="1,1,1"
   ) {
    rotationMode=BMM_Replace
    rotationSpace=BCS_WorldSpace
    scaleSpace=BCS_BoneSpace
    translationMode=BMM_Replace
    translationSpace=BCS_WorldSpace
    alphaBoolBlend {blendOption=Linear}
    alphaScaleBiasClamp {inRange {max=0, min=-40}, outRange {max=0, min=1}}
    boneToModify {boneName=body_calf_r}
   }
   Save cached pose 'Body Transforms'(Pose=Component To Local(ComponentPose=@n13)) {cacheName=Body Transforms}
   n14 = "Transform (Modify) Bone - Bone: body_root"(
    ComponentPose=Local To Component(LocalPose=Use cached pose 'Body Transforms'())
    Alpha=1
    Translation="0,200,0"
    Rotation="0,0,0"
    Scale="1,1,1"
   ) {translationMode=BMM_Additive, translationSpace=BCS_BoneSpace, boneToModify {boneName=body_root}, translation {y=200}}
   Save cached pose 'Pre Dynamics setup'(Pose=Component To Local(ComponentPose=@n14)) {cacheName=Pre Dynamics setup}
  ]}
 EventGraph {nodes [
   n1 = "Gravity Override State Function (Parameter Asset) (Nativized)"(CharacterPartAnimInstance=Self-Reference(), GameTimeInSeconds=GetGameTimeInSeconds())
   > n2 = Set GravityOverride(GravityOverride=@n1)
   n3 = GetPartSkeletalMeshComponent()
   Is Valid(exec=[@n2.then, @n15.then], InputObject=@n3)
   n4 = GetAttachParent!(self:Scene Component Object Reference=@n3) -> (ReturnValue:Scene Component Object Reference)
   > Is Valid(InputObject=@n4)
   > Set Body_Mesh(Body_Mesh:Scene Component Object Reference=@n4) -> (Output_Get:Scene Component Object Reference)
   > n5 = Sequence!(execute) -> (then_0, then_1, then_2:Exec, then_3:Exec, then_4:Exec, then_5:Exec, then_6:Exec)
   n6 = Get Body_Mesh() -> (Body_Mesh:Scene Component Object Reference)
   n7 = BreakTransform(InTransform=GetSocketTransform(self=@n6, InSocketName=root, TransformSpace=RTS_World))
   Set root_location(execute=@n5.then_6, root_location:Vector=@n7) -> (Output_Get:Vector)
   > Set root_space_rotation(root_space_rotation:Rotator=@n7.Rotation) -> (Output_Get:Rotator)
   n8 = BreakTransform(InTransform=GetSocketTransform(self=@n6, InSocketName=pelvis, TransformSpace=RTS_World))
   Set pelvis_location(execute=@n5.then_0, pelvis_location:Vector=@n8) -> (Output_Get:Vector)
   > Set pelvis_rotation(pelvis_rotation:Rotator=@n8.Rotation) -> (Output_Get:Rotator)
   n9 = BreakTransform(InTransform=GetSocketTransform(self=@n6, InSocketName=thigh_l, TransformSpace=RTS_World))
   Set thigh_left_location(execute=@n5.then_1, thigh_left_location:Vector=@n9) -> (Output_Get:Vector)
   > Set thigh_left_rotation(thigh_left_rotation:Rotator=@n9.Rotation) -> (Output_Get:Rotator)
   n10 = BreakTransform(InTransform=GetSocketTransform(self=@n6, InSocketName=thigh_r, TransformSpace=RTS_World))
   Set thigh_right_location(execute=@n5.then_2, thigh_right_location:Vector=@n10) -> (Output_Get:Vector)
   > Set thigh_right_rotation(thigh_right_rotation:Rotator=@n10.Rotation) -> (Output_Get:Rotator)
   n11 = BreakTransform(InTransform=GetSocketTransform(self=@n6, InSocketName=calf_r, TransformSpace=RTS_World))
   Set calf_right_location(execute=@n5.then_4, calf_right_location:Vector=@n11) -> (Output_Get:Vector)
   > Set calf_right_rotation(calf_right_rotation:Rotator=@n11.Rotation) -> (Output_Get:Rotator)
   n12 = BreakTransform(InTransform=GetSocketTransform(self=@n6, InSocketName=calf_l, TransformSpace=RTS_World))
   Set calf_left_location(execute=@n5.then_3, calf_left_location:Vector=@n12) -> (Output_Get:Vector)
   > Set calf_left_rotation(calf_left_rotation:Rotator=@n12.Rotation) -> (Output_Get:Rotator)
   n13 = BreakTransform(InTransform=GetSocketTransform(self=@n6, InSocketName=root, TransformSpace=RTS_World))
   Set body_root_location(execute=@n5.then_5, body_root_location:Vector=@n13) -> (Output_Get:Vector)
   > Set body_root_rotation(body_root_rotation:Rotator=@n13.Rotation) -> (Output_Get:Rotator)
   Event BlueprintUpdateAnimation()
   > n14 = Sequence!(execute) -> (then_0, then_1, then_2)
   > n15 = Set is Visible but NOT Equipped(is Visible but NOT Equipped:bool=NOT Boolean(A=Get bIsOnPawn() -> (bIsOnPawn:bool))) -> (Output_Get:bool)
   n16 = Get is Visible but NOT Equipped()
   n17 = Branch(execute=@n14.then_1, Condition=@n16)
   > Set RBAN Alpha(RBAN Alpha:double=0.5) -> (Output_Get:double)
   Set RBAN Alpha(execute=@n17.else, RBAN Alpha=1)
   n18 = Branch(execute=@n14.then_2, Condition=@n16)
   > Set GravityOverride(GravityOverride="0,0,-300")
   Set GravityOverride(execute=@n18.else, GravityOverride="0,0,-980")
   TryGetPawnOwner()
   Get Gravity Override Params() -> (Gravity Override Params:Fort Gravity Override Parameter Set Structure)
  ]}
]
variables [
 Body_Mesh {type=SceneComponent, default=null}
 body_root_location {type=Vector}
 body_root_rotation {type=Rotator}
 calf_left_location {type=Vector}
 calf_left_rotation {type=Rotator}
 calf_right_location {type=Vector}
 calf_right_rotation {type=Rotator}
 Gravity Override Params {
  type=FortGravityOverrideParameterSet
  default=(Skydive_GoingUp=(GravityOverride=()),Skydive_FloatingDown=(GravityOverride=()),Skydive_DivingDown=(GravityOverride=()),Skydive_Parachuting=(GravityOverride=()),OnGround_Standing=(GravityOverride=()),OnGround_Moving=(GravityOverride=()),OnGround_Moving_HighRotationRate=(GravityOverride=()),OnGround_Crouch_Moving=(GravityOverride=()),Emote_or_Melee=(GravityOverride=()),Swimming=(GravityOverride=()),Front_End=(GravityOverride=()),DBNO=(GravityOverride=()),DBNO_Carried=(GravityOverride=()),OnGround_Falling=(GravityOverride=()),Ziplining=(GravityOverride=()),Skydive_Surfing=(GravityOverride=()),Riding_Standing=(GravityOverride=()),Riding_Moving=(GravityOverride=()),Riding_Moving_Targeting=(GravityOverride=()),Tactical_Sprint=(GravityOverride=()),Rail_Grinding=(GravityOverride=()),)
 }
 GravityOverride {type=Vector}
 is Visible but NOT Equipped {type=boolean}
 pelvis_location {type=Vector}
 pelvis_rotation {type=Rotator}
 RBAN Alpha {type=number}
 root_location {type=Vector}
 root_space_rotation {type=Rotator}
 thigh_left_location {type=Vector}
 thigh_left_rotation {type=Rotator}
 thigh_right_location {type=Vector}
 thigh_right_rotation {type=Rotator}
]
